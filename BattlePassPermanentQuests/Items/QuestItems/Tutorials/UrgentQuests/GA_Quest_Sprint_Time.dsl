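; Gameplay ability for a battle pass tutorial quest that tracks how long the player sprints.
; On event-triggered activation it calls the parent, waits for the configured seconds-to-sprint, sends an Athena.Quests.Sprint.OverTime custom stat event to the owning player's quest manager, then ends.
; The graph itself checks no sprint state during the delay.
; Ending the ability only prints a debug string.
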(event Ability|EventActivateAbilityFromEvent
  (params (EventData :type Gameplay%20Event%20Data%20Structure%20%28by%20ref%29))
  (|CallParentFunction EventData)
  (Utilities|FlowControl|Delay (Variables|Default|GetSecondsToSprint))
  (Quest|SendComplexCustomStatEvent (Variables|GAVars|Data|GetOwningPlayerQuestManager) 0 "(GameplayTags=((TagName=\"Athena.Quests.Sprint.OverTime\")))" "(GameplayTags=)" 1)
  (Ability|EndAbility))

(event Ability|EventOnEndAbility
  (params (bWasCancelled :type Boolean))
  (Development|PrintString "ending ability..." true true "(R=1.000000,G=0.000000,B=0.163166,A=1.000000)" 3.0))
