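; Transient quest gameplay ability for a clamber tutorial urgent quest, adjusting the quest description to match the player's auto-clamber setting.
; Activation stores the avatar as the owning player pawn.
; Post-initialization calls UpdateClamberDescOverride, then starts a five-second delay commented as a periodic recheck, though no loop back is shown.
; Off the server only, that function reads the setting and passes a MoveForward tutorial tag or empty value to the transient quests component.
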(event Ability|EventActivateAbility
  (bind _asfort_player_pawn_athena (Utilities|Casting|CastToFortPlayerPawnAthena (Ability|GetAvatarActorfromActorInfo)))
  (Variables|Default|SetOwning%20Player _asfort_player_pawn_athena))

(event Ability|EventFailedToActivatePassiveAbility
  (params (ActorInfo :type Gameplay%20Ability%20Actor%20Info%20Structure))
  (|CallParentFunction))

(event EventOnPostAbilityInitialization
  (CallFunction|UpdateClamberDescOverride)
  (Utilities|FlowControl|Delay 5.0) ; Check occasionally in case the user changes the setting
)

(fn UpdateClamberDescOverride
  (if (not (Networking|IsServer))
    (bind _asfort_urgent_quest_item_definition (Utilities|Casting|CastToFortUrgentQuestItemDefinition (Variables|FortTransientQuestGameplayAbility|GetQuestItemDefinition)))
    (Class|FortControllerComponentTransientQuests|SetUrgentQuestDescriptionOverride (Class|FortPlayerPawn|GetTransientQuestsComponent (Variables|Default|GetOwning%20Player)) (Utilities|Struct|BreakUrgentQuestData (Class|FortUrgentQuestItemDefinition|GetUrgentQuestData _asfort_urgent_quest_item_definition)) (select (FortGameOptions|GetAutoClamberEnabled (BlueprintContext|GetFortSettingsContext)) "(TagName=\"SpecialEvent.Tutorial.02.MoveForward\")" 0))))
